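# Sound cue for the swing of the Modern Military pickaxe cosmetic, primed on load and limited per owner by a stop-oldest concurrency override.
# Mixes a close-attenuated layer of Military swings with a medium-attenuated layer of borrowed DefaultV2 distant swings.
# Each layer's quality-level switch scales from a random pool through a smaller pool to a single wave.
# Modulators randomize volume on both layers and raise pitch on the close one.
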SoundCue /BRCosmetics/Sounds/PickAxes/ModernMilitary/PickAxe_ModernMilitary_Swing_Athena_Cue
root=Mixer_3
nodes [
 {id=Attenuation_6, children [Modulator_8], properties {attenuationSettings=/Game/Sounds/Attenuation/Fort_Weapon_Pickaxe_Close_Att}}
 {id=Attenuation_7, children [Modulator_7], properties {attenuationSettings=/Game/Sounds/Attenuation/Fort_Weapon_Pickaxe_Medium_Att}}
 {id=Mixer_3, children [Attenuation_6, Attenuation_7], properties {inputVolume [1,1]}}
 {id=Modulator_10, children [Random_6], properties {volumeMax=0.9, volumeMin=0.8}}
 {id=Modulator_11, children [WavePlayer_37], properties {volumeMax=0.9, volumeMin=0.8}}
 {id=Modulator_12, children [Random_7], properties {pitchMax=1.2, pitchMin=1.1, volumeMax=1, volumeMin=0.9}}
 {id=Modulator_13, children [WavePlayer_38], properties {pitchMax=1.2, pitchMin=1.1, volumeMax=1, volumeMin=0.9}}
 {id=Modulator_3, children [Random_1], properties {pitchMax=1.2, pitchMin=1.1, volumeMax=1, volumeMin=0.9}}
 {id=Modulator_7, children [QualityLevel_2]}
 {id=Modulator_8, children [QualityLevel_1]}
 {id=Modulator_9, children [Random_5], properties {volumeMax=0.9, volumeMin=0.8}}
 {id=QualityLevel_1, children [Modulator_3,Modulator_12,Modulator_13,Modulator_13]}
 {id=QualityLevel_2, children [Modulator_9,Modulator_10,Modulator_11,Modulator_11]}
 {
  id=Random_1
  children [WavePlayer_0, WavePlayer_7, WavePlayer_8, WavePlayer_9, WavePlayer_10, WavePlayer_11]
  properties {weights [1,1,1,1,1,1]}
 }
 {id=Random_5, children [WavePlayer_3, WavePlayer_6, WavePlayer_14, WavePlayer_30, WavePlayer_31], properties {weights [1,1,1,1,1]}}
 {id=Random_6, children [WavePlayer_32, WavePlayer_35, WavePlayer_36], properties {weights [1,1,1]}}
 {id=Random_7, children [WavePlayer_15, WavePlayer_33, WavePlayer_34], properties {weights [1,1,1]}}
 {id=WavePlayer_0, properties {soundWaveAssetPtr=&/PickAxe_Military_Swing_01}}
 {id=WavePlayer_10, properties {soundWaveAssetPtr=&/PickAxe_Military_Swing_05}}
 {id=WavePlayer_11, properties {soundWaveAssetPtr=&/PickAxe_Military_Swing_06}}
 {id=WavePlayer_14, properties {soundWaveAssetPtr=/BRCosmetics/Sounds/PickAxes/DefaultV2/Pickaxe_DefaultV2_Swing_Distant_03}}
 {id=WavePlayer_15, properties {soundWaveAssetPtr=&PickAxe_Military_Swing_01}}
 {id=WavePlayer_3, properties {soundWaveAssetPtr=/BRCosmetics/Sounds/PickAxes/DefaultV2/Pickaxe_DefaultV2_Swing_Distant_01}}
 {id=WavePlayer_30, properties {soundWaveAssetPtr=/BRCosmetics/Sounds/PickAxes/DefaultV2/Pickaxe_DefaultV2_Swing_Distant_04}}
 {id=WavePlayer_31, properties {soundWaveAssetPtr=/BRCosmetics/Sounds/PickAxes/DefaultV2/Pickaxe_DefaultV2_Swing_Distant_05}}
 {id=WavePlayer_32, properties {soundWaveAssetPtr=&Pickaxe_DefaultV2_Swing_Distant_01}}
 {id=WavePlayer_33, properties {soundWaveAssetPtr=&/PickAxe_Military_Swing_03}}
 {id=WavePlayer_34, properties {soundWaveAssetPtr=&/PickAxe_Military_Swing_02}}
 {id=WavePlayer_35, properties {soundWaveAssetPtr=&Pickaxe_DefaultV2_Swing_Distant_04}}
 {id=WavePlayer_36, properties {soundWaveAssetPtr=&Pickaxe_DefaultV2_Swing_Distant_05}}
 {id=WavePlayer_37, properties {soundWaveAssetPtr=&Pickaxe_DefaultV2_Swing_Distant_04}}
 {id=WavePlayer_38, properties {soundWaveAssetPtr=&PickAxe_Military_Swing_01}}
 {id=WavePlayer_6, properties {soundWaveAssetPtr=/BRCosmetics/Sounds/PickAxes/DefaultV2/Pickaxe_DefaultV2_Swing_Distant_02}}
 {id=WavePlayer_7, properties {soundWaveAssetPtr=&PickAxe_Military_Swing_03}}
 {id=WavePlayer_8, properties {soundWaveAssetPtr=&PickAxe_Military_Swing_02}}
 {id=WavePlayer_9, properties {soundWaveAssetPtr=&/PickAxe_Military_Swing_04}}
]
properties {
 audioPropertiesSheet=/BRCosmetics/Sounds/Settings/PropertySheets/Pickaxe/APS_Cosmetic_Pickaxe_Swing
 bOverrideConcurrency=true
 bPrimeOnLoad=true
 maxDistance=3000
 soundClassObject=&Fort_WeaponMelee_PSM
 volumeMultiplier=0.3
 concurrencyOverrides {bLimitToOwner=true, maxCount=2, resolutionRule=StopOldest}
 concurrencySet [&Player_Melee_Weapon_Conc]
}
